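# Translucent unlit hologram material for the Elastic character's Niagara mesh particles, showing an emoji texture as a still or looping flipbook.
# World-aligned panning noise, gated by layered sine flicker offset by actor position, plus a TurnOn height pulse, modulate brightness, opacity and horizontal vertex offset.
# World-Z scanlines scale with camera distance and shape opacity, alongside optional strobe and particle alpha.
# Low quality zeroes the noise term for color and opacity.
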Material /BRCosmetics/Characters/Player/Female/Medium/Bodies/F_MED_Elastic/FX/M_Elastic_Holograms
graph {
 properties {
  bUseMaterialAttributes=true
  bUsedWithNiagaraMeshParticles=true
  blendMode=BLEND_Translucent
  pixelDepthOffsetMode=PDOM_Legacy
  shadingModel=MSM_Unlit
  twoSided=true
  usedShadingModels=Unlit
 }
 nodes [
  n1 = MaterialFunctionCall(UVs=TextureCoordinate(), Texture Scale=ScalarParameter() {defaultValue=0.75, parameterName=UV Scale}) {materialFunction=&ScaleUVsByCenter}
  n2 = TextureObjectParameter() {group=Texture, parameterName=2DHologramTexture, texture=/BRCosmetics/2dAssets/Emoji/Season22/S22_ChillCat_Claw_Sprite}
  n3 = TextureSample(UVs=@n1.UVs, Tex=@n2) {samplerSource=SSM_Clamp_WorldGroupSettings, texture=/BRCosmetics/2dAssets/Emoji/Season4/babySeal}
  n4 = Time()
  n5 = ScalarParameter() {defaultValue=4, parameterName=Rows, sliderMax=4}
  n6 = ScalarParameter() {defaultValue=4, parameterName=Columns, sliderMax=4}
  n7 = Multiply(A=Multiply(A=@n5, B=@n6), B=Constant() {desc=0.1 sec per frame, r=0.1})
  n8 = MaterialFunctionCall!(
   "Animation  Phase (0-1)"=Frac(None=Divide(A=@n4, B=@n7))
   Number of Rows=@n5
   Number of Columns=@n6
   Texture=@n2
   UVs=@n1.UVs
   MipBias/Level
   "Use Mip Bias (T) Level (F)"
   "Clamp Anim (see tooltip)"
  ) {materialFunction=/BRCosmetics/Characters/Player/Female/Medium/Bodies/F_MED_ZombieElasticNeon/FX/FlipBook_Clamped}
  n9 = ScalarParameter() {defaultValue=1, parameterName=AnimatedHologram}
  n10 = Multiply(A=Min(A=LinearInterpolate(A=@n3.A, B=@n8.Alpha, Alpha=@n9) {desc=A}, B=@n1.0-1 mask), B=@n1.0-1 mask)
  n11 = ScalarParameter() {defaultValue=1, group=Noise, parameterName=NoiseFresnelXPMult}
  n12 = Divide(A=ObjectRadius(), B=ScalarParameter() {defaultValue=0.1, group=Noise, parameterName=NoiseSize})
  n13 = Multiply(A=Time(), B=VectorParameter() {group=Noise, parameterName=NoisePanSpeed, defaultValue=<rgba -0.08 0.07 -0.2 0>}.RGB)
  n14 = MaterialFunctionCall(
   TextureObject=TextureObject() {samplerType=SAMPLERTYPE_LinearColor, texture=&T_NoisePack_LinearColor}
   TextureSize=Multiply(A=VectorParameter() {group=Noise, parameterName=NoiseTileMult, defaultValue=<rgba 1 1 0.1 1>}.RGB, B=@n12)
   WorldPosition=Add(A=WorldPosition().XYZ, B=Multiply(A=@n12, B=@n13))
  ) {materialFunction=&WorldAlignedTexture}
  n15 = ActorPositionWS()
  n16 = MaterialFunctionCall(Float2=Round(None=Divide(A=@n15) {constB=2000})) {materialFunction=&BreakOutFloat2Components}
  n17 = Add(A=Time(), B=Add(A=@n16.R, B=@n16.R))
  n18 = Multiply(A=ScalarParameter() {defaultValue=0.5, parameterName=FlickerSpeed}, B=@n17)
  n19 = AppendVector(A=Sine(None=@n18) {period=1.6}, B=Sine(None=@n18) {period=1.42})
  n20 = Sine(None=@n18) {period=1.51}
  n21 = Sine(None=@n18) {period=2.1}
  n22 = ScalarParameter() {defaultValue=0.5, group=Noise, parameterName=NoiseFrequency}
  n23 = Multiply(B=@n22) {constA=0.25}
  n24 = Add(A=Sine(None=@n18) {period=0.72}, B=Multiply(B=@n22) {constA=0.75}) {constB=0.75}
  n25 = Add(A=Multiply(B=@n22) {constA=0.9, constB=0.9}, B=Sine(None=@n18) {period=1.25}) {constB=0.9}
  n26 = Multiply(A=Multiply(A=Add(A=@n21, B=@n23), B=@n24), B=Multiply(A=@n24, B=Ceil(None=@n25)))
  n27 = ConstantBiasScale(None=Sine(None=@n17) {period=0.075})
  n28 = Multiply(A=Ceil(None=AppendVector(A=@n19, B=@n20)), B=Multiply(A=Saturate(None=@n26), B=LinearInterpolate(Alpha=@n27) {constA=0.7}))
  n29 = ScalarParameter() {defaultValue=2, group=Noise, parameterName=NoiseMult}
  n30 = Multiply(
   A=Multiply(A=Power(Base=@n14.XY Texture) {constExponent=4}, B=ScalarParameter() {defaultValue=2, group=Noise, parameterName=NoiseMaskMult}) {constB=20}
   B=Add(A=ScalarParameter() {group=Noise, parameterName=NoiseAdd}, B=Multiply(A=@n28, B=@n29))
  )
  n31 = ComponentMask(None=MaterialFunctionCall() {materialFunction=&ObjectPivotPoint}.Object Pivot Location) {b=true}
  n32 = ScalarParameter() {defaultValue=300, parameterName=PulseHeight}
  n33 = ScalarParameter() {defaultValue=1, parameterName=TurnOn}
  n34 = ScalarParameter() {defaultValue=33, parameterName=PulseRadius}
  n35 = SphereMask(
   A=ComponentMask(None=WorldPosition().XYZ) {b=true}
   B=LinearInterpolate(A=@n31, B=Add(A=@n31, B=@n32) {constB=1600}, Alpha=@n33) {constA=-0.25, constB=1.2}
   Radius=LinearInterpolate(A=@n34, B=Multiply(A=@n34) {constB=0.1}, Alpha=@n33) {constA=500, constB=100}
  ) {hardnessPercent=0}
  n36 = OneMinus(None=@n33)
  n37 = Add(A=Desaturation(None=Saturate(None=@n30)), B=Multiply(A=Multiply(A=@n35, B=@n36)) {constB=2})
  n38 = QualitySwitch(
   Default=MaterialFunctionCall(In=@n37, Contrast=ScalarParameter() {defaultValue=0.2, group=Noise, parameterName=NoiseContrast}) {materialFunction=&CheapContrast}.Result
   Low=Constant()
  )
  n39 = Power(
   Base=Desaturation(
    None=LinearInterpolate(A=@n3.RGB, B=@n8.Result, Alpha=@n9) {desc=RGB}
    Fraction=ScalarParameter() {defaultValue=0.15, group=Texture, parameterName=Desaturation}
   )
   Exp=Multiply(A=ScalarParameter() {defaultValue=2, parameterName=FresnelExp}, B=LinearInterpolate(B=@n11, Alpha=@n38) {constA=1, constB=0.5})
  )
  n40 = Saturate(None=@n38)
  n41 = Multiply(A=ScalarParameter() {group=Noise, parameterName=NoiseBrightness}, B=@n40) {constB=0.1}
  LinearInterpolate(A=Multiply(A=@n10, B=Add(A=@n39, B=@n41)), B=@n10) {constAlpha=1}
  n42 = ParticleColor()
  n43 = Multiply(A=Multiply(A=@n39, B=ScalarParameter() {defaultValue=2, parameterName=EmissiveMult}) {constB=2}, B=@n42.RGB)
  n44 = ScalarParameter() {parameterName=UseParticleColor}
  n45 = Multiply(A=@n38) {constB=1.5}
  n46 = Clamp(None=Subtract(A=ScalarParameter() {defaultValue=0.2, parameterName=MinScanlineOpacity}, B=@n45)) {minDefault=-0.25}
  n47 = Add(A=@n45)
  n48 = ScalarParameter() {defaultValue=400, parameterName=ScanlineTile}
  n49 = LinearInterpolate(Alpha=Ceil(None=@n40)) {constA=1, constB=1.25}
  n50 = ComponentMask(None=WorldPosition().XYZ) {b=true}
  n51 = ScalarParameter() {defaultValue=20000, parameterName=ScaleDistance}
  n52 = Distance(A=CameraPositionWS(), B=ActorPositionWS())
  n53 = LinearInterpolate(B=@n51, Alpha=Saturate(None=Divide(A=@n52, B=@n51) {constB=1000})) {constA=50, constB=100}
  n54 = Sine(None=Multiply(A=Multiply(A=@n48, B=@n49), B=Divide(A=@n50, B=@n53) {constB=20}) {constA=1})
  n55 = Abs(None=DotProduct(A=VertexNormalWS(), B=Constant3Vector() {constant=<rgba 0 0 1 0>}))
  n56 = Power(Base=LinearInterpolate(A=ConstantBiasScale(None=@n54), Alpha=Power(Base=@n55) {constExponent=5}) {constB=0.25}) {constExponent=1.5}
  n57 = Saturate(None=Multiply(A=@n10, B=LinearInterpolate(A=@n46, B=@n47, Alpha=@n56) {constA=0.5}))
  n58 = ScalarParameter() {defaultValue=1.2, parameterName=OpacityMult}
  n59 = Time()
  n60 = ScalarParameter() {defaultValue=5, parameterName=StrobeSpeed}
  n61 = Multiply(A=Sine(None=Multiply(A=@n59, B=@n60)), B=ScalarParameter() {defaultValue=2, parameterName=StrobeHardness})
  n62 = Multiply(
   A=Multiply(A=LinearInterpolate(B=@n42.A, Alpha=@n44) {constA=1, desc=Workaround for skel mesh bug}, B=Multiply(A=@n57, B=@n58))
   B=LinearInterpolate(B=Clamp(None=@n61), Alpha=ScalarParameter() {parameterName=Strobe}) {constA=1}
  )
  n63 = Normalize(VectorInput=Subtract(A=WorldPosition().XYZ, B=ObjectPositionWS()))
  n64 = Constant3Vector() {constant=<rgba 1 1 0 0>}
  n65 = ScalarParameter() {defaultValue=10, group=Noise, parameterName=WPOStrength}
  n66 = Multiply(A=ObjectRadius(), B=ScalarParameter() {defaultValue=0.025, group=Noise, parameterName=WPOStrengthFromObjectRadius}) {constB=0.05}
  n67 = Multiply(A=@n37, B=Multiply(A=Multiply(A=@n63, B=@n64), B=Add(A=@n65, B=@n66))) {constB=0}
  MaterialFunctionCall(In=SetMaterialAttributes!(
    MaterialAttributes
    Emissive Color=NamedRerouteDeclaration(None=@n43) {name=emissive, nodeColor=<rgba 1 0 0.8203 1>}
    Opacity=NamedRerouteDeclaration(None=@n62) {name=opacity, nodeColor=<rgba 1 0.4922 0 1>}
    World Position Offset=NamedRerouteDeclaration(None=@n67) {name=wpo, nodeColor=<rgba 1 0 0.5391 1>}
   ) {attributeSetTypes [#1, #2, #3]}) {materialFunction=&MF_TechArtOverrides}
  ScalarParameter() {parameterName=AnimationPhase}
 ]
}
