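# Niagara effect for the Elastic female character's backpack, layering a sticker mesh over a darkening sprite.
# Darken bursts one translucent black sprite with fixed facing and dynamic material parameters, persisting until the emitter deactivates.
# Sticker bursts one mesh particle on a short loop cycle.
# Both sit slightly offset from the system position and toward the camera; user parameters expose glow and hologram materials, both empty by default.
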NiagaraSystem /BRCosmetics/Characters/Player/Female/Medium/Bodies/F_MED_Elastic/FX/NS_Elastic_Backpack
dependencies {usedDataInterfaces [], usedDynamicInputs [], usedModules [&SystemState]}
emitters {
 Darken {
  emitterUpdateScript [
   &EmitterState(Life Cycle Mode=::NewEnumerator1, Loop Behavior=::NewEnumerator1, Loop Duration Mode=::NewEnumerator1, UseLoopDelay=false)
   &SpawnBurst_Instantaneous(Spawn Count=1, Use Spawn Probability=false)
  ]
  particleSpawnScript [
   /Niagara/Modules/Spawn/Initialization/InitializeParticle(
    Color=<rgba 0 0 0 0.45>
    Module.Write Color=true
    Module.Write Lifetime=false
    Module.Write Mass=true
    Module.Write Material Random=true
    Module.Write Position=true
    Module.Write Scale=false
    Module.Write SpriteRotation=false
    Module.Write SpriteSize=true
    Position=&Add_Vector(A=/Niagara/DynamicInputs/Helpers/SimulationPosition(), B=<0 -8.5 0>)
    Sprite Size=<35 35>
   ) {enabled=false}
   /Niagara/Modules/Spawn/Initialization/V2/InitializeParticle(
    Color=<rgba 0 0 0 0.45>
    Mass Mode=::NewEnumerator1
    Position=&Add_Vector(A=&SimulationPosition(), B=<0 -8.5 0>)
    Position Mode=::NewEnumerator1
    Sprite Size=<35 35>
    Sprite Size Mode=::NewEnumerator4
    UsePositionOffset=false
    Write Material Random=true
   ) {module=InitializeParticle001}
   &DynamicMaterialParameters(
    Index 0 Param 1=1.2
    Index 0 Param 2=0.15
    Index 0 Param 3=1
    Index 0 Param 4=1
    Module.Param0WriteEnabled=true
    Module.Param1WriteEnabled=true
    Module.Param2WriteEnabled=true
    Module.Param3WriteEnabled=true
   )
   &CameraOffset(Camera Offset Amount=10)
   /NiagaraFN/Sprite/FN_SpriteAlignmentAndFacing(SpriteAlignment:Vector3f=<1 0 0>, SpriteFacing:Vector3f=<0 -1 0>) {module=SpriteAlignmentAndFacing}
  ]
  particleUpdateScript [&ParticleState(Kill Particles When Lifetime Has Elapsed=false, Let Infinitely Lived Particles Die When Emitter Deactivates=true)]
  rendererClasses [&NiagaraSpriteRendererProperties]
  renderers [{rendererClass=&NiagaraSpriteRendererProperties}]
 }
 Sticker {
  emitterUpdateScript [
   &EmitterState(Life Cycle Mode=::NewEnumerator1, Loop Duration=0.9, UseLoopDelay=false)
   &SpawnBurst_Instantaneous(Spawn Count=1, Use Spawn Probability=false)
  ]
  particleSpawnScript [
   /Niagara/Modules/Spawn/Initialization/InitializeParticle(
    Module.Write Color=true
    Module.Write Lifetime=true
    Module.Write Mass=true
    Module.Write Material Random=true
    Module.Write Position=true
    Module.Write Scale=true
    Module.Write SpriteRotation=false
    Module.Write SpriteSize=false
    Position=&Add_Vector(A=&SimulationPosition(), B=<0 -8 0>)
    Sprite Size=<5 5>
   ) {enabled=false}
   /Niagara/Modules/Spawn/Initialization/V2/InitializeParticle(
    Mass Mode=::NewEnumerator1
    Mesh Scale Mode=::NewEnumerator4
    Position=&Add_Vector(A=&SimulationPosition(), B=<0 -8 0>)
    Position Mode=::NewEnumerator1
    Sprite Size Mode=::NewEnumerator4
    UsePositionOffset=false
    Write Material Random=true
   ) {module=InitializeParticle001}
   /Niagara/Modules/Spawn/Orientation/V2/InitialMeshOrientation(Rotation=<0 0 -0.25>, Use Orientation Vector=false, Use Rotation Vector=true) {enabled=false}
   /Niagara/Modules/Spawn/Orientation/V3/InitialMeshOrientation(Mesh Orientation Mode=::NewEnumerator1, Rotation=<0 0 -0.25>, Use Rotation Vector=true) {module=InitialMeshOrientation001}
   &CameraOffset(Camera Offset Amount=10)
  ]
  particleUpdateScript [
   &ParticleState()
   &SolveForcesAndVelocity(Clamp Velocity=false, Limit Acceleration=false)
  ]
  rendererClasses [&NiagaraMeshRendererProperties]
  renderers [{rendererClass=&NiagaraMeshRendererProperties, rendererIndex=0}]
 }
}
summary {
 systemName=NS_Elastic_Backpack
 emitters [
  {emitterName=Sticker, rendererClasses [&NiagaraMeshRendererProperties]}
  {emitterName=Darken, rendererClasses [&NiagaraSpriteRendererProperties]}
 ]
 userVariables [
  User.MID_ElasticGlow {
   type=MaterialInterface
   description=""
   defaultValue {struct=NiagaraExt_VariableValue_Object, value {object=null, objectClass=/Script/Engine.MaterialInterface}}
  }
  User.MID_ElasticHologram {
   type=MaterialInterface
   description=""
   defaultValue {struct=NiagaraExt_VariableValue_Object, value {object=null, objectClass=&MaterialInterface}}
  }
 ]
}
