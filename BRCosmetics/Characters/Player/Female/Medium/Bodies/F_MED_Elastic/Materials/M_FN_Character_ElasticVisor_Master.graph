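# Clear-coat master material for the visor of the Elastic female character, built on the shared PreFX and PostFX character functions.
# AccessoryColor is darkened through the FX mask's blue channel, while SpecularMasks scaled by MSR drive metallic, specular and roughness, before a Clearcoat function.
# The Hide parameter scales a World Position Offset that collapses vertices toward the object's position.
# A Fresnel-based visor opacity chain exists but feeds no output.
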Material /BRCosmetics/Characters/Player/Female/Medium/Bodies/F_MED_Elastic/Materials/M_FN_Character_ElasticVisor_Master
graph {
 properties {
  bCastRayTracedShadows=false
  bUseMaterialAttributes=true
  bUsedWithMeshDeformer=true
  bUsedWithMorphTargets=true
  bUsedWithSkeletalMesh=true
  pixelDepthOffsetMode=PDOM_Legacy
  shadingModel=MSM_ClearCoat
  translucencyLightingMode=TLM_Surface
  usedShadingModels=Clear Coat
 }
 nodes [
  n1 = MaterialFunctionCall() {materialFunction=&BaseFN_CharacterFunction_PreFX}
  n2 = VectorParameter() {group=Elastic, parameterName=AccessoryColor, sortPriority=0}
  n3 = TextureSampleParameter2D() {
   group=Elastic
   parameterName=FX
   samplerType=SAMPLERTYPE_Masks
   texture=/BRCosmetics/Characters/Player/Female/Medium/Bodies/F_MED_Elastic/Skins/Patterns/Textures/Mask/T_F_MED_Elastic_Mask_Pattern_1_FX
  }
  n4 = TextureSampleParameter2D() {
   group=Elastic
   parameterName=SpecularMasks
   samplerType=SAMPLERTYPE_Masks
   sortPriority=0
   texture=/Game/Characters/Player/Female/Medium/Bodies/F_MED_Elastic/Textures/T_F_MED_Elastic_Body_S
  }
  n5 = VectorParameter() {group=Elastic, parameterName=MSR, defaultValue=<rgba 0 0.8 0.3 0>}
  n6 = Subtract(A=WorldPosition().XYZ, B=ObjectPositionWS())
  n7 = SetMaterialAttributes!(
   MaterialAttributes=@n1.Result
   Base Color=LinearInterpolate(A=Multiply(A=@n2.RGB) {constB=0.1}, B=@n2.RGB, Alpha=@n3.B)
   Metallic=Multiply(A=@n4.G, B=@n5.R)
   Specular=Multiply(A=@n4.R, B=@n5.G)
   Roughness=Multiply(A=@n4.B, B=@n5.B)
   World Position Offset=Multiply(A=Multiply(A=@n6) {constB=-1}, B=ScalarParameter() {defaultValue=1, group=Elastic, parameterName=Hide})
   Opacity
  ) {attributeSetTypes [
    #1
    #2
    #3
    #4
    #5
    #6
   ]}
  MaterialFunctionCall(
   AO=@n1.AO
   Distance Blend=@n1.AO
   Fresnel=@n1.AO
   Gmap=@n1.AO
   Material=MaterialFunctionCall!(Alpha=ScalarParameter() {defaultValue=1}, In=@n7) {materialFunction=/Game/Weapons/WeaponSkins/MaterialFunction/Clearcoat}.Result
  ) {materialFunction=&BaseFN_CharacterFunction_PostFX}
  n8 = ScalarParameter() {defaultValue=1, group=Elastic, parameterName=VisorExponent, sliderMax=10}
  n9 = ScalarParameter() {defaultValue=0.1, group=Elastic, parameterName=VisorBaseOpacity, sliderMax=1}
  Saturate(None=Multiply(
    A=ScalarParameter() {defaultValue=0.5, group=Elastic, parameterName=VisorOpactiy, sliderMax=1}
    B=Fresnel(ExponentIn=@n8, BaseReflectFractionIn=@n9) {baseReflectFraction=0.1, exponent=1}
   ))
 ]
}
